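# Niagara system for the Snow Ninja pickaxe that adds a persistent blue glow sprite and cyan lights.
# Its single emitter burst-spawns a particle at a fixed offset, nudged by camera offset, that outlives its lifetime until the emitter deactivates.
# The particle feeds a sprite renderer and two light renderers, with two offset light positions written at spawn.
# A time-of-day scalar adjusts the particle's alpha.
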NiagaraSystem /BRCosmetics/Weapons/FORT_Melee/Pickaxe_Snow_Ninja/FX/NS_SnowNinja_light
dependencies {usedDataInterfaces [], usedDynamicInputs [], usedModules [&SystemState]}
emitters {Glow_and_Light {
  emitterUpdateScript [
   &EmitterState(
    Inactive Response=::NewEnumerator1
    Life Cycle Mode=::NewEnumerator1
    Loop Behavior=::NewEnumerator1
    Loop Duration Mode=::NewEnumerator1
    UseLoopDelay=false
   )
   &SpawnBurst_Instantaneous(Use Spawn Probability=false, UseLoopCountLimit=false)
  ]
  particleSpawnScript [
   &InitializeParticle(Color=<rgba 0 0.4066 1 0.7>, Sprite Size Mode=::NewEnumerator3, Uniform Sprite Size=36, UsePositionOffset=false, Write Material Random=true)
   &CameraOffset(Camera Offset Amount=10)
   &DynamicMaterialParameters(
    Index 0 Param 1=0.8
    Index 0 Param 2=2
    Index 0 Param 3=1
    Index 0 Param 4=1
    Module.Param0WriteEnabled=true
    Module.Param1WriteEnabled=true
    Module.Param2WriteEnabled=true
    Module.Param3WriteEnabled=true
   )
   &SystemLocation(MaskSpawnGroup=false, Offset=<0 6 50>)
   /SetVariables(Particles.Light_position:Vector3f=&Add_Vector(A=Linked() {linkedVariable {name=Particles.Position, type=NiagaraPosition}}, B=<5 3 0>)) {isSetParameters=true}
   /SetVariables(Particles.Light_position02:Vector3f=&Add_Vector(A=Linked() {linkedVariable {name=Particles.Position, type=NiagaraPosition}}, B=<-10 5 0>)) {isSetParameters=true}
   /SetVariables(Particles.LightExponent=2) {isSetParameters=true}
   /SetVariables(Particles.LightRadius=20) {isSetParameters=true}
   /SetVariables(Particles.Light_color:LinearColor=<rgba 0 3.999 5 1>) {isSetParameters=true}
  ]
  particleUpdateScript [
   &ParticleState(Kill Particles When Lifetime Has Elapsed=false, Let Infinitely Lived Particles Die When Emitter Deactivates=true)
   &ScaleColor(
    Scale Alpha=&Multiply_Float(A=1, B=&FNTimeOfDayScalar(Night Scale=0.3, Morning Scale=0.2, Evening Scale=0.3))
    ScaleA=true
    ScaleRGB=true
   )
  ]
  rendererClasses [&NiagaraSpriteRendererProperties, &NiagaraLightRendererProperties]
  renderers [
   {rendererClass=&NiagaraSpriteRendererProperties}
   {rendererClass=&NiagaraLightRendererProperties, rendererIndex=1}
   {rendererClass=&NiagaraLightRendererProperties, rendererIndex=2}
  ]
 }}
summary {
 systemName=NS_SnowNinja_light
 emitters [{
   emitterName=Glow_and_Light
   rendererClasses [&NiagaraSpriteRendererProperties, &NiagaraLightRendererProperties]
  }]
 userVariables []
}
