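# Post-input AnimBlueprint for the Snow Ninja pickaxe that simulates its two hanging ribbons with per-bone Anim Dynamics.
# The cached input pose feeds three chains using gravity, angular springs and spherical limits on the ribbon tips.
# In the front end, bIsOnPawn selects between two stiff, tightly limited chains, each pre-rotating the ribbon base bones.
# Outside it, a looser chain runs without base offsets; the event graph has no logic.
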AnimBlueprint /BRCosmetics/Weapons/FORT_Melee/Pickaxe_Snow_Ninja/Meshes/SK_Pickaxe_Snow_Ninja_AnimBP
target_skeleton=&/SK_Pickaxe_Snow_Ninja_Skeleton
graphs [
 AnimGraph {nodes [
   n1 = "Transform (Modify) Bone - Bone: based_ribbon_up"(ComponentPose=Local To Component(LocalPose=Use cached pose 'presim'()), Alpha=1, Translation="0,0,0", Rotation="-27,66,0", Scale="1,1,1") {rotationMode=BMM_Additive, rotationSpace=BCS_BoneSpace, boneToModify {boneName=based_ribbon_up}, rotation {pitch=-27, yaw=66}}
   n2 = "Transform (Modify) Bone - Bone: based_ribbon_dn"(ComponentPose=@n1, Alpha=1, Translation="0,0,0", Rotation="-44,66,0", Scale="1,1,1") {rotationMode=BMM_Additive, rotationSpace=BCS_BoneSpace, boneToModify {boneName=based_ribbon_dn}, rotation {pitch=-44, yaw=66}}
   n3 = Anim Dynamics - Bone: dyn_ribbon_dn_1(ComponentPose=@n2, Alpha=1, ExternalForce="0,0,0") {
    angularDampingOverride=1
    angularSpringConstant=1200
    bAngularSpring=true
    bOverrideAngularDamping=true
    bUseGravityOverride=true
    bUsePlanarLimit=false
    boundBone {boneName=dyn_ribbon_dn_1}
    componentLinearAccScale=<0.3 0.3 0.3>
    componentLinearVelScale=<0.3 0.3 0.3>
    gravityOverride {z=-980}
    physicsBodyDefinitions [{
      boundBone {boneName=dyn_ribbon_dn_1}
      boxExtents=<8 8 8>
      constraintSetup {angularLimitsMax=<11 11 11>, angularLimitsMin=<-11 -11 -11>, angularTarget=<0 0 0>}
      localJointOffset=<1 0 0>
     }]
    simSpaceSettings {maxAngularAcceleration=7, maxAngularVelocity=35, simSpaceAngularAlpha=1}
   }
   n4 = Anim Dynamics - Bone: dyn_ribbon_dn_2(ComponentPose=@n3, Alpha=1, ExternalForce="0,0,0") {
    angularDampingOverride=1
    angularSpringConstant=1200
    bAngularSpring=true
    bOverrideAngularDamping=true
    bUseGravityOverride=true
    bUsePlanarLimit=false
    boundBone {boneName=dyn_ribbon_dn_2}
    componentLinearAccScale=<0.3 0.3 0.3>
    componentLinearVelScale=<0.3 0.3 0.3>
    gravityOverride {z=-980}
    physicsBodyDefinitions [{
      boundBone {boneName=dyn_ribbon_dn_2}
      boxExtents=<8 8 8>
      constraintSetup {angularLimitsMax=<11 11 11>, angularLimitsMin=<-11 -11 -11>, angularTarget=<0 0 0>}
      localJointOffset=<1 0 0>
     }]
    simSpaceSettings {maxAngularAcceleration=7, maxAngularVelocity=35, simSpaceAngularAlpha=1}
   }
   n5 = Anim Dynamics - Bone: dyn_ribbon_dn_3(ComponentPose=@n4, Alpha=1, ExternalForce="0,0,0") {
    angularDampingOverride=1
    angularSpringConstant=1200
    bAngularSpring=true
    bOverrideAngularDamping=true
    bUseGravityOverride=true
    bUsePlanarLimit=false
    bUseSphericalLimits=true
    boundBone {boneName=dyn_ribbon_dn_3}
    componentLinearAccScale=<0.3 0.3 0.3>
    componentLinearVelScale=<0.3 0.3 0.3>
    gravityOverride {z=-980}
    physicsBodyDefinitions [{
      collisionType=CustomSphere
      sphereCollisionRadius=7
      boundBone {boneName=dyn_ribbon_dn_3}
      boxExtents=<8 8 8>
      constraintSetup {angularLimitsMax=<11 11 11>, angularLimitsMin=<-11 -11 -11>, angularTarget=<0 0 0>}
      localJointOffset=<1 0 0>
     }]
    simSpaceSettings {maxAngularAcceleration=7, maxAngularVelocity=35, simSpaceAngularAlpha=1}
    sphericalLimits [{limitRadius=-9.7135, drivingBone {boneName=None}, sphereLocalOffset=<56.8827 6.4881 -68.6695>}]
   }
   n6 = Anim Dynamics - Bone: dyn_ribbon_dn_4(ComponentPose=@n5, Alpha=1, ExternalForce="0,0,0") {
    angularDampingOverride=1
    angularSpringConstant=1200
    bAngularSpring=true
    bOverrideAngularDamping=true
    bUseGravityOverride=true
    bUsePlanarLimit=false
    bUseSphericalLimits=true
    boundBone {boneName=dyn_ribbon_dn_4}
    componentLinearAccScale=<0.3 0.3 0.3>
    componentLinearVelScale=<0.3 0.3 0.3>
    gravityOverride {z=-980}
    physicsBodyDefinitions [{
      collisionType=CustomSphere
      sphereCollisionRadius=7
      boundBone {boneName=dyn_ribbon_dn_4}
      boxExtents=<8 8 8>
      constraintSetup {angularLimitsMax=<11 11 11>, angularLimitsMin=<-11 -11 -11>, angularTarget=<0 0 0>}
      localJointOffset=<1 0 0>
     }]
    simSpaceSettings {maxAngularAcceleration=7, maxAngularVelocity=35, simSpaceAngularAlpha=1}
    sphericalLimits [{limitRadius=7.5009, drivingBone {boneName=Melee_TwoHanded_Handle}, sphereLocalOffset=<58.0408 -1.6394 -31.6896>}]
   }
   n7 = Anim Dynamics - Bone: dyn_ribbon_up_1(ComponentPose=@n6, Alpha=1, ExternalForce="0,0,0") {
    angularDampingOverride=1
    angularSpringConstant=1200
    bAngularSpring=true
    bOverrideAngularDamping=true
    bUseGravityOverride=true
    bUsePlanarLimit=false
    boundBone {boneName=dyn_ribbon_up_1}
    componentLinearAccScale=<0.3 0.3 0.3>
    componentLinearVelScale=<0.3 0.3 0.3>
    gravityOverride {z=-980}
    physicsBodyDefinitions [{
      boundBone {boneName=dyn_ribbon_up_1}
      boxExtents=<8 8 8>
      constraintSetup {angularLimitsMax=<33 33 33>, angularLimitsMin=<-33 -33 -33>, angularTarget=<0 0 0>}
      localJointOffset=<1 0 0>
     }]
    simSpaceSettings {maxAngularAcceleration=7, maxAngularVelocity=35, simSpaceAngularAlpha=1}
   }
   n8 = Anim Dynamics - Bone: dyn_ribbon_up_2(ComponentPose=@n7, Alpha=1, ExternalForce="0,0,0") {
    angularDampingOverride=1
    angularSpringConstant=1200
    bAngularSpring=true
    bOverrideAngularDamping=true
    bUseGravityOverride=true
    bUsePlanarLimit=false
    boundBone {boneName=dyn_ribbon_up_2}
    componentLinearAccScale=<0.3 0.3 0.3>
    componentLinearVelScale=<0.3 0.3 0.3>
    gravityOverride {z=-980}
    physicsBodyDefinitions [{
      boundBone {boneName=dyn_ribbon_up_2}
      boxExtents=<8 8 8>
      constraintSetup {angularLimitsMax=<33 33 33>, angularLimitsMin=<-33 -33 -33>, angularTarget=<0 0 0>}
      localJointOffset=<1 0 0>
     }]
    simSpaceSettings {maxAngularAcceleration=7, maxAngularVelocity=35, simSpaceAngularAlpha=1}
   }
   n9 = Anim Dynamics - Bone: dyn_ribbon_up_3(ComponentPose=@n8, Alpha=1, ExternalForce="0,0,0") {
    angularDampingOverride=1
    angularSpringConstant=1200
    bAngularSpring=true
    bOverrideAngularDamping=true
    bUseGravityOverride=true
    bUsePlanarLimit=false
    bUseSphericalLimits=true
    boundBone {boneName=dyn_ribbon_up_3}
    componentLinearAccScale=<0.3 0.3 0.3>
    componentLinearVelScale=<0.3 0.3 0.3>
    gravityOverride {z=-980}
    physicsBodyDefinitions [{
      collisionType=CustomSphere
      sphereCollisionRadius=7
      boundBone {boneName=dyn_ribbon_up_3}
      boxExtents=<8 8 8>
      constraintSetup {angularLimitsMax=<33 33 33>, angularLimitsMin=<-33 -33 -33>, angularTarget=<0 0 0>}
      localJointOffset=<1 0 0>
     }]
    simSpaceSettings {maxAngularAcceleration=7, maxAngularVelocity=35, simSpaceAngularAlpha=1}
    sphericalLimits [{limitRadius=5.5108, drivingBone {boneName=None}, sphereLocalOffset=<-3.2988 5.8139 5.8011>}]
   }
   n10 = Component To Local(ComponentPose=@n9)
   n11 = "Transform (Modify) Bone - Bone: based_ribbon_up"(
    ComponentPose=Local To Component(LocalPose=Use cached pose 'presim'())
    Alpha=1
    Translation="0,0,0"
    Rotation="-4.376,-62.359,48.632"
    Scale="1,1,1"
   ) {rotationMode=BMM_Additive, rotationSpace=BCS_BoneSpace, boneToModify {boneName=based_ribbon_up}, rotation=<rot -4.376 -62.359 48.632>}
   n12 = "Transform (Modify) Bone - Bone: based_ribbon_dn"(ComponentPose=@n11, Alpha=1, Translation="0,0,0", Rotation="-0.003,-23.021,31.086", Scale="1,1,1") {rotationMode=BMM_Additive, rotationSpace=BCS_BoneSpace, boneToModify {boneName=based_ribbon_dn}, rotation=<rot -0.003 -23.021 31.086>}
   n13 = Anim Dynamics - Bone: dyn_ribbon_dn_1(ComponentPose=@n12, Alpha=1, ExternalForce="0,0,0") {
    angularDampingOverride=1
    angularSpringConstant=700
    bAngularSpring=true
    bOverrideAngularDamping=true
    bUseGravityOverride=true
    bUsePlanarLimit=false
    boundBone {boneName=dyn_ribbon_dn_1}
    componentLinearAccScale=<0.3 0.3 0.3>
    componentLinearVelScale=<0.3 0.3 0.3>
    gravityOverride {z=-980}
    physicsBodyDefinitions [{
      boundBone {boneName=dyn_ribbon_dn_1}
      boxExtents=<5 5 5>
      constraintSetup {angularLimitsMax=<11 11 11>, angularLimitsMin=<-11 -11 -11>, angularTarget=<0 0 0>}
      localJointOffset=<1 0 0>
     }]
    simSpaceSettings {maxAngularAcceleration=7, maxAngularVelocity=35, simSpaceAngularAlpha=1}
   }
   n14 = Anim Dynamics - Bone: dyn_ribbon_dn_2(ComponentPose=@n13, Alpha=1, ExternalForce="0,0,0") {
    angularDampingOverride=1
    angularSpringConstant=700
    bAngularSpring=true
    bOverrideAngularDamping=true
    bUseGravityOverride=true
    bUsePlanarLimit=false
    boundBone {boneName=dyn_ribbon_dn_2}
    componentLinearAccScale=<0.3 0.3 0.3>
    componentLinearVelScale=<0.3 0.3 0.3>
    gravityOverride {z=-980}
    physicsBodyDefinitions [{
      boundBone {boneName=dyn_ribbon_dn_2}
      boxExtents=<5 5 5>
      constraintSetup {angularLimitsMax=<11 11 11>, angularLimitsMin=<-11 -11 -11>, angularTarget=<0 0 0>}
      localJointOffset=<1 0 0>
     }]
    simSpaceSettings {maxAngularAcceleration=7, maxAngularVelocity=35, simSpaceAngularAlpha=1}
   }
   n15 = Anim Dynamics - Bone: dyn_ribbon_dn_3(ComponentPose=@n14, Alpha=1, ExternalForce="0,0,0") {
    angularDampingOverride=1
    angularSpringConstant=700
    bAngularSpring=true
    bOverrideAngularDamping=true
    bUseGravityOverride=true
    bUsePlanarLimit=false
    bUseSphericalLimits=true
    boundBone {boneName=dyn_ribbon_dn_3}
    componentLinearAccScale=<0.3 0.3 0.3>
    componentLinearVelScale=<0.3 0.3 0.3>
    gravityOverride {z=-980}
    physicsBodyDefinitions [{
      collisionType=CustomSphere
      sphereCollisionRadius=7
      boundBone {boneName=dyn_ribbon_dn_3}
      boxExtents=<5 5 5>
      constraintSetup {angularLimitsMax=<11 11 11>, angularLimitsMin=<-11 -11 -11>, angularTarget=<0 0 0>}
      localJointOffset=<1 0 0>
     }]
    simSpaceSettings {maxAngularAcceleration=7, maxAngularVelocity=35, simSpaceAngularAlpha=1}
    sphericalLimits [{limitRadius=15.7208, drivingBone {boneName=None}, sphereLocalOffset=<15.8112 11.4515 -72.4262>}]
   }
   n16 = Anim Dynamics - Bone: dyn_ribbon_dn_4(ComponentPose=@n15, Alpha=1, ExternalForce="0,0,0") {
    angularDampingOverride=1
    angularSpringConstant=700
    bAngularSpring=true
    bOverrideAngularDamping=true
    bUseGravityOverride=true
    bUsePlanarLimit=false
    bUseSphericalLimits=true
    boundBone {boneName=dyn_ribbon_dn_4}
    componentLinearAccScale=<0.3 0.3 0.3>
    componentLinearVelScale=<0.3 0.3 0.3>
    gravityOverride {z=-980}
    physicsBodyDefinitions [{
      collisionType=CustomSphere
      sphereCollisionRadius=7
      boundBone {boneName=dyn_ribbon_dn_4}
      boxExtents=<5 5 5>
      constraintSetup {angularLimitsMax=<11 11 11>, angularLimitsMin=<-11 -11 -11>, angularTarget=<0 0 0>}
      localJointOffset=<1 0 0>
     }]
    simSpaceSettings {maxAngularAcceleration=7, maxAngularVelocity=35, simSpaceAngularAlpha=1}
    sphericalLimits [{limitRadius=15.7208, drivingBone {boneName=Melee_TwoHanded_Handle}, sphereLocalOffset=<12.4399 6.9391 -35.1942>}]
   }
   n17 = Anim Dynamics - Bone: dyn_ribbon_up_1(ComponentPose=@n16, Alpha=1, ExternalForce="0,0,0") {
    angularDampingOverride=1
    angularSpringConstant=700
    bAngularSpring=true
    bOverrideAngularDamping=true
    bUseGravityOverride=true
    bUsePlanarLimit=false
    boundBone {boneName=dyn_ribbon_up_1}
    componentLinearAccScale=<0.3 0.3 0.3>
    componentLinearVelScale=<0.3 0.3 0.3>
    gravityOverride {z=-980}
    physicsBodyDefinitions [{
      boundBone {boneName=dyn_ribbon_up_1}
      boxExtents=<5 5 5>
      constraintSetup {angularLimitsMax=<22 11 11>, angularLimitsMin=<-22 -12 -12>, angularTarget=<0 0 0>}
      localJointOffset=<1 0 0>
     }]
    simSpaceSettings {maxAngularAcceleration=7, maxAngularVelocity=35, simSpaceAngularAlpha=1}
   }
   n18 = Anim Dynamics - Bone: dyn_ribbon_up_2(ComponentPose=@n17, Alpha=1, ExternalForce="0,0,0") {
    angularDampingOverride=1
    angularSpringConstant=700
    bAngularSpring=true
    bOverrideAngularDamping=true
    bUseGravityOverride=true
    bUsePlanarLimit=false
    boundBone {boneName=dyn_ribbon_up_2}
    componentLinearAccScale=<0.3 0.3 0.3>
    componentLinearVelScale=<0.3 0.3 0.3>
    gravityOverride {z=-980}
    physicsBodyDefinitions [{
      boundBone {boneName=dyn_ribbon_up_2}
      boxExtents=<5 5 5>
      constraintSetup {angularLimitsMax=<22 11 11>, angularLimitsMin=<-22 -12 -12>, angularTarget=<0 0 0>}
      localJointOffset=<1 0 0>
     }]
    simSpaceSettings {maxAngularAcceleration=7, maxAngularVelocity=35, simSpaceAngularAlpha=1}
   }
   n19 = Anim Dynamics - Bone: dyn_ribbon_up_3(ComponentPose=@n18, Alpha=1, ExternalForce="0,0,0") {
    angularDampingOverride=1
    angularSpringConstant=700
    bAngularSpring=true
    bOverrideAngularDamping=true
    bUseGravityOverride=true
    bUsePlanarLimit=false
    bUseSphericalLimits=true
    boundBone {boneName=dyn_ribbon_up_3}
    componentLinearAccScale=<0.3 0.3 0.3>
    componentLinearVelScale=<0.3 0.3 0.3>
    gravityOverride {z=-980}
    physicsBodyDefinitions [{
      collisionType=CustomSphere
      sphereCollisionRadius=7
      boundBone {boneName=dyn_ribbon_up_3}
      boxExtents=<5 5 5>
      constraintSetup {angularLimitsMax=<22 11 11>, angularLimitsMin=<-22 -12 -12>, angularTarget=<0 0 0>}
      localJointOffset=<1 0 0>
     }]
    simSpaceSettings {maxAngularAcceleration=7, maxAngularVelocity=35, simSpaceAngularAlpha=1}
    sphericalLimits [{limitRadius=15.7208, drivingBone {boneName=None}, sphereLocalOffset=<11.1304 16.222 7.5598>}]
   }
   n20 = Component To Local(ComponentPose=@n19)
   n21 = Get bIsOnPawn() -> (bIsOnPawn:bool)
   n22 = Anim Dynamics - Bone: dyn_ribbon_dn_1(ComponentPose=Local To Component(LocalPose=Use cached pose 'presim'()), Alpha=1, ExternalForce="0,0,0") {
    angularDampingOverride=0.9
    angularSpringConstant=150
    bAngularSpring=true
    bOverrideAngularDamping=true
    bUseGravityOverride=true
    bUsePlanarLimit=false
    boundBone {boneName=dyn_ribbon_dn_1}
    componentLinearAccScale=<0.3 0.3 0.3>
    componentLinearVelScale=<0.3 0.3 0.3>
    gravityOverride {z=-980}
    physicsBodyDefinitions [{
      boundBone {boneName=dyn_ribbon_dn_1}
      boxExtents=<13 13 13>
      constraintSetup {angularLimitsMax=<22 22 55>, angularLimitsMin=<-55 -55 -55>, angularTarget=<0 0 0>}
      localJointOffset=<1 0 0>
     }]
    simSpaceSettings {maxAngularAcceleration=7, maxAngularVelocity=35, simSpaceAngularAlpha=1}
   }
   n23 = Anim Dynamics - Bone: dyn_ribbon_dn_2(ComponentPose=@n22, Alpha=1, ExternalForce="0,0,0") {
    angularDampingOverride=0.9
    angularSpringConstant=150
    bAngularSpring=true
    bOverrideAngularDamping=true
    bUseGravityOverride=true
    bUsePlanarLimit=false
    boundBone {boneName=dyn_ribbon_dn_2}
    componentLinearAccScale=<0.3 0.3 0.3>
    componentLinearVelScale=<0.3 0.3 0.3>
    gravityOverride {z=-980}
    physicsBodyDefinitions [{
      boundBone {boneName=dyn_ribbon_dn_2}
      boxExtents=<13 13 13>
      constraintSetup {angularLimitsMax=<22 22 33>, angularLimitsMin=<-20 -20 -33>, angularTarget=<0 0 0>}
      localJointOffset=<1 0 0>
     }]
    simSpaceSettings {maxAngularAcceleration=7, maxAngularVelocity=35, simSpaceAngularAlpha=1}
   }
   n24 = Anim Dynamics - Bone: dyn_ribbon_dn_3(ComponentPose=@n23, Alpha=1, ExternalForce="0,0,0") {
    angularDampingOverride=0.9
    angularSpringConstant=150
    bAngularSpring=true
    bOverrideAngularDamping=true
    bUseGravityOverride=true
    bUsePlanarLimit=false
    bUseSphericalLimits=true
    boundBone {boneName=dyn_ribbon_dn_3}
    componentLinearAccScale=<0.3 0.3 0.3>
    componentLinearVelScale=<0.3 0.3 0.3>
    gravityOverride {z=-980}
    physicsBodyDefinitions [{
      collisionType=CustomSphere
      sphereCollisionRadius=7
      boundBone {boneName=dyn_ribbon_dn_3}
      boxExtents=<13 13 13>
      constraintSetup {angularLimitsMax=<22 22 22>, angularLimitsMin=<-20 -20 -22>, angularTarget=<0 0 0>}
      localJointOffset=<1 0 0>
     }]
    simSpaceSettings {maxAngularAcceleration=7, maxAngularVelocity=35, simSpaceAngularAlpha=1}
    sphericalLimits [{limitRadius=15.7208, drivingBone {boneName=None}, sphereLocalOffset=<15.8112 11.4515 -72.4262>}]
   }
   n25 = Anim Dynamics - Bone: dyn_ribbon_dn_4(ComponentPose=@n24, Alpha=1, ExternalForce="0,0,0") {
    angularDampingOverride=0.9
    angularSpringConstant=150
    bAngularSpring=true
    bOverrideAngularDamping=true
    bUseGravityOverride=true
    bUsePlanarLimit=false
    bUseSphericalLimits=true
    boundBone {boneName=dyn_ribbon_dn_4}
    componentLinearAccScale=<0.3 0.3 0.3>
    componentLinearVelScale=<0.3 0.3 0.3>
    gravityOverride {z=-980}
    physicsBodyDefinitions [{
      collisionType=CustomSphere
      sphereCollisionRadius=7
      boundBone {boneName=dyn_ribbon_dn_4}
      boxExtents=<13 13 13>
      constraintSetup {angularLimitsMax=<22 22 11>, angularLimitsMin=<-20 -20 -11>, angularTarget=<0 0 0>}
      localJointOffset=<1 0 0>
     }]
    simSpaceSettings {maxAngularAcceleration=7, maxAngularVelocity=35, simSpaceAngularAlpha=1}
    sphericalLimits [{limitRadius=15.7208, drivingBone {boneName=Melee_TwoHanded_Handle}, sphereLocalOffset=<12.4399 6.9391 -35.1942>}]
   }
   n26 = Anim Dynamics - Bone: dyn_ribbon_up_1(ComponentPose=@n25, Alpha=1, ExternalForce="0,0,0") {
    angularDampingOverride=0.9
    angularSpringConstant=150
    bAngularSpring=true
    bOverrideAngularDamping=true
    bUseGravityOverride=true
    bUsePlanarLimit=false
    boundBone {boneName=dyn_ribbon_up_1}
    componentLinearAccScale=<0.3 0.3 0.3>
    componentLinearVelScale=<0.3 0.3 0.3>
    gravityOverride {z=-980}
    physicsBodyDefinitions [{
      boundBone {boneName=dyn_ribbon_up_1}
      boxExtents=<13 13 13>
      constraintSetup {angularLimitsMax=<22 22 55>, angularLimitsMin=<-55 -55 -55>, angularTarget=<0 0 0>}
      localJointOffset=<1 0 0>
     }]
    simSpaceSettings {maxAngularAcceleration=7, maxAngularVelocity=35, simSpaceAngularAlpha=1}
   }
   n27 = Anim Dynamics - Bone: dyn_ribbon_up_2(ComponentPose=@n26, Alpha=1, ExternalForce="0,0,0") {
    angularDampingOverride=0.9
    angularSpringConstant=150
    bAngularSpring=true
    bOverrideAngularDamping=true
    bUseGravityOverride=true
    bUsePlanarLimit=false
    boundBone {boneName=dyn_ribbon_up_2}
    componentLinearAccScale=<0.3 0.3 0.3>
    componentLinearVelScale=<0.3 0.3 0.3>
    gravityOverride {z=-980}
    physicsBodyDefinitions [{
      boundBone {boneName=dyn_ribbon_up_2}
      boxExtents=<13 13 13>
      constraintSetup {angularLimitsMax=<22 22 33>, angularLimitsMin=<-20 -20 -33>, angularTarget=<0 0 0>}
      localJointOffset=<1 0 0>
     }]
    simSpaceSettings {maxAngularAcceleration=7, maxAngularVelocity=35, simSpaceAngularAlpha=1}
   }
   n28 = Anim Dynamics - Bone: dyn_ribbon_up_3(ComponentPose=@n27, Alpha=1, ExternalForce="0,0,0") {
    angularDampingOverride=0.9
    angularSpringConstant=150
    bAngularSpring=true
    bOverrideAngularDamping=true
    bUseGravityOverride=true
    bUsePlanarLimit=false
    bUseSphericalLimits=true
    boundBone {boneName=dyn_ribbon_up_3}
    componentLinearAccScale=<0.3 0.3 0.3>
    componentLinearVelScale=<0.3 0.3 0.3>
    gravityOverride {z=-980}
    physicsBodyDefinitions [{
      collisionType=CustomSphere
      sphereCollisionRadius=7
      boundBone {boneName=dyn_ribbon_up_3}
      boxExtents=<13 13 13>
      constraintSetup {angularLimitsMax=<22 22 22>, angularLimitsMin=<-20 -20 -22>, angularTarget=<0 0 0>}
      localJointOffset=<1 0 0>
     }]
    simSpaceSettings {maxAngularAcceleration=7, maxAngularVelocity=35, simSpaceAngularAlpha=1}
    sphericalLimits [{limitRadius=15.7208, drivingBone {boneName=None}, sphereLocalOffset=<11.1304 16.222 7.5598>}]
   }
   Output Pose(Result=Blend Poses by bool(
     BlendPose_0=Blend Poses by bool(BlendPose_0=@n20, BlendPose_1=@n10, bActiveValue=@n21)
     BlendPose_1=Component To Local(ComponentPose=@n28)
     bActiveValue=Get bIsInFrontEnd()
    )) {layerGroup=None}
   AnimGraphNode_SubInput_0 = Input Pose()
   Save cached pose 'presim'(Pose=@AnimGraphNode_SubInput_0) {cacheName=presim}
  ]}
 EventGraph {nodes [Event BlueprintUpdateAnimation()]}
]
variables []
