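# Physics asset for the Snow Ninja pickaxe skeletal mesh, a two-handed melee cosmetic with dangling ribbons.
# The handle body combines four capsules and anchors two ribbon chains, an upper one of three bones and a lower one of four.
# Each chain starts with a sphere at the root, followed by thin boxes that approximate flat ribbon segments.
# Angular-limited constraints link each segment to its parent, with wider swing and twist at the handle joints than along the chains.
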PhysicsAsset /BRCosmetics/Weapons/FORT_Melee/Pickaxe_Snow_Ninja/Meshes/SK_Pickaxe_Snow_Ninja_Physics
bodies {
 Melee_TwoHanded_Handle [
  {shapeType=Capsule, length=120.0755, radius=2.8383, center=<0.5942 -1.3263 18.2477>, rotation=<rot 0.282 -90.0022 0.0684>}
  {shapeType=Capsule, length=1.75, radius=4.5, center=<-0.0002 -1.4483 -39.7442>, rotation=<rot 0 0 90.0005>}
  {shapeType=Capsule, length=33.75, radius=15.5, center=<9.6956 0 17.2653>}
  {shapeType=Capsule, length=15, radius=16.25, center=<0 6.653 -46.5125>}
 ]
 dyn_ribbon_dn_1 [{shapeType=Sphere, radius=2.75, center=<0 -1.1206 0>}]
 dyn_ribbon_dn_2 [{shapeType=Box, extentX=0.5, extentY=9.5, extentZ=3, center=<0 0.8279 -0.0001>}]
 dyn_ribbon_dn_3 [{shapeType=Box, extentX=0.5, extentY=12, extentZ=3, center=<0 -0.2241 -0.0001>}]
 dyn_ribbon_dn_4 [{shapeType=Box, extentX=0.5, extentY=11, extentZ=3, center=<0 0.0752 -0.0001>}]
 dyn_ribbon_up_1 [{shapeType=Sphere, radius=3.75, center=<0 1.427 -0.0001>}]
 dyn_ribbon_up_2 [{shapeType=Box, extentX=0.5, extentY=5, extentZ=2.5, center=<0 -0.5362 -0.0001>}]
 dyn_ribbon_up_3 [{shapeType=Box, extentX=0.5, extentY=9, extentZ=2.5, center=<0 -1.9453 0.0001>}]
}
constraints [
 {bone1Name=dyn_ribbon_up_1, bone2Name=Melee_TwoHanded_Handle, swing1LimitDegrees=65, swing2LimitDegrees=45, twistLimitDegrees=45}
 {bone1Name=dyn_ribbon_dn_1, bone2Name=Melee_TwoHanded_Handle, swing1LimitDegrees=65, swing2LimitDegrees=45, twistLimitDegrees=45}
 {bone1Name=dyn_ribbon_dn_2, bone2Name=dyn_ribbon_dn_1, swing1LimitDegrees=60, swing2LimitDegrees=25, twistLimitDegrees=25}
 {bone1Name=dyn_ribbon_dn_3, bone2Name=dyn_ribbon_dn_2, swing1LimitDegrees=60, swing2LimitDegrees=25, twistLimitDegrees=25}
 {bone1Name=dyn_ribbon_dn_4, bone2Name=dyn_ribbon_dn_3, swing1LimitDegrees=60, swing2LimitDegrees=25, twistLimitDegrees=25}
 {bone1Name=dyn_ribbon_up_2, bone2Name=dyn_ribbon_up_1, swing1LimitDegrees=60, swing2LimitDegrees=30, twistLimitDegrees=30}
 {bone1Name=dyn_ribbon_up_3, bone2Name=dyn_ribbon_up_2, swing1LimitDegrees=60, swing2LimitDegrees=20, twistLimitDegrees=20}
]
